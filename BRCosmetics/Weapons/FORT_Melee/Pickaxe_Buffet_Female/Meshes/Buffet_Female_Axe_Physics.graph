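# Physics asset for the Buffet Female pickaxe, a two-handed melee cosmetic weapon mesh.
# It holds a single capsule body on the Melee_TwoHanded_Handle bone, offset and rotated to cover the axe.
# There are no constraints, so the weapon simulates and collides as one rigid piece.
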PhysicsAsset /BRCosmetics/Weapons/FORT_Melee/Pickaxe_Buffet_Female/Meshes/Buffet_Female_Axe_Physics
bodies {Melee_TwoHanded_Handle [{shapeType=Capsule, length=76.2213, radius=22.3786, center=<0.7606 -0.9979 29.7185>, rotation=<rot 0.0163 -90 0.0172>}]}
constraints []
